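; Creative building prop Blueprint for an Apollo urban street light, named as the NoLight variant.
; It overrides the day phase change event, calls the parent, then prints a debug string naming the current phase.
; The construction script is empty and the file contains no light-switching logic of its own.
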(event EventOnDayPhaseChanged
  (params (CurrentDayPhase :type EFortDayPhase%20Enum :default "Morning") (PreviousDayPhase :type EFortDayPhase%20Enum :default "Morning") (bAtCreation :type Boolean))
  (|CallParentFunction CurrentDayPhase PreviousDayPhase bAtCreation)
  (Development|PrintString (Utilities|String|Append "Time has changed " (Utilities|String|EnumtoString CurrentDayPhase))))

(fn ConstructionScript)
