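; Blueprint for a Creative building prop, the SVLair geode object, with no logic of its own in the graphs shown.
; Its construction script is empty.
; An OnRep_IsRocketFinal replication handler is present but also empty, so changes to that replicated rocket state trigger no behavior here.
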(fn OnRep_IsRocketFinal)

(fn ConstructionScript)
