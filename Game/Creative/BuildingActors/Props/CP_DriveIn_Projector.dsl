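; Creative drive-in projector prop Blueprint that toggles a spotlight and a projector particle/effect component through a replicated integer ProjectorState.
; OnRep_ProjectorState sets the light intensity and activates or deactivates the effect for the off and on states.
; On BeginPlay the authority either turns it on via a debug flag or binds BeginScript to the game mode's match-started event.
; BeginScript is empty, so only that flag and two debug events change the state here.
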(event EventBeginPlay
  ;[1 0 0] beginplay
  (Networking|SwitchHasAuthority
    (:Authority
      ;[1 1 1] debug?
      (if (Variables|Default|GetProjectorEnabledOnBeginplay)
        (Variables|Default|SetProjectorState 1)
        (else
          ;[1 1 1] Wait for match started.
          (bind _asfort_game_mode_zone (Utilities|Casting|CastToFortGameModeZone (Game|GetGameMode)))
          (EventDispatchers|BindEventtoOnHandleMatchHasStarted _asfort_game_mode_zone (EventDispatchers|CreateEvent :!event Custom|BeginScript)))))))

(event Custom|projectordebug_1
  ;[1 1 1] debug
  (Variables|Default|SetProjectorState 1))

(event Custom|BeginScript)

(event Custom|projectordebug_0
  ;[1 1 1] debug
  (Variables|Default|SetProjectorState 0))

(fn OnRep_ProjectorState
  (bind _spotlight (Variables|Default|GetSpotLight))
  (bind _p_drivein_projectoron (Variables|Default|GetP_DriveIn_ProjectorON))
  (switch int (Variables|Default|GetProjectorState)
    (:0
      (Rendering|Components|Light|SetIntensity _spotlight 0.0)
      (Components|Activation|Deactivate _p_drivein_projectoron))
    (:1
      (Rendering|Components|Light|SetIntensity _spotlight 5.0)
      (Components|Activation|Activate _p_drivein_projectoron false))))

(fn ConstructionScript)
