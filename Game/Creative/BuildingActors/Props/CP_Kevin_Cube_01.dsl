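; Creative prop version of the Kevin cube: a building actor that bounces players and vehicles, never dies, and hosts weak spots.
; Its CubeTurnedOn rep-notify toggles ambient audio and, on authority, spawns or destroys an attached overlap actor that plays bounce FX.
; Authority-side helpers attach shuffled weak spots to sockets, avoiding the last-used set, and drop random items or coins.
; Other helpers damage overlapping building actors and build random ground-traced movement paths.
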(fn CleanupOnDestroyed
  (bind _overlapactor (Variables|Default|GetOverlapActor))
  ;[1 1 1] Destroy the Weak Spot Actors
  (CallFunction|DestroyWeakSpots)
  ;[1 1 1] Destroy the Overlap Actor
  (Utilities|IsValid _overlapactor
    (:Is%20Valid
      (Actor|DestroyActor _overlapactor))))

(fn DestroyOverlapingBuildingActors
  (bind (_outactors _returnvalue) (Collision|BoxOverlapActors (Transformation|GetActorLocation) "800.000000,800.000000,800.000000" (Utilities|Array|MakeArray "ObjectTypeQuery1") "/Script/FortniteGame.BuildingActor" (Utilities|Array|MakeArray self) :!class /Script/Engine.KismetSystemLibrary))
  (if _returnvalue
    (for _array_element _outactors
      (Utilities|IsValid _array_element
        (:Is%20Valid
          (Development|PrintString (Utilities|GetDisplayName _array_element) false false)
          (Environment|ApplyGlobalEnvironmentDamagetoActor _array_element 1e+03 "(GameplayTags=((TagName=\"Athena.Gameplay.Status.ForceSkipDBNO\"),(TagName=\"Gameplay.Damage.Environment.Cube\")))"))))))

(fn DestroyWeakSpots
  (bind _spawnedweakspotactors (Variables|Default|GetSpawnedWeakSpotActors))
  (Networking|SwitchHasAuthority
    (:Authority
      (if (Utilities|IsArrayNonEmpty _spawnedweakspotactors)
        (for _array_element _spawnedweakspotactors
          (Utilities|IsValid _array_element
            (:Is%20Valid
              (Class|HyperCubeWeakSpot|RemoveWeakSpot _array_element))))
        (Utilities|Array|Clear _spawnedweakspotactors)
        (else
          (Development|PrintString "NO WEAK SPOTS" false false))))))

(fn DropItems
  (bind _itemstodrop (Variables|Default|GetItemsToDrop))
  (bind _returnvalue (Utilities|Array|Length _itemstodrop))
  (Networking|SwitchHasAuthority
    (:Authority
      (if (> _returnvalue 0)
        (Inventory|SpawnPickupInWorld (Utilities|Array|Get(acopy) _itemstodrop (Math|Random|RandomIntegerinRange 0 (- _returnvalue 1))) (Math|Random|RandomIntegerinRange (Variables|Default|GetDropCountMin) (Variables|Default|GetDropCountMax)) (Transformation|GetSocketLocation (Variables|Default|GetStaticMesh) (Utilities|Array|Get(acopy) (Variables|Default|GetWeakSpotSockets) (Math|Random|RandomIntegerinRange 0 (- (Utilities|Array|Length (Variables|Default|GetWeakSpotSockets)) 1))))) ; Pistol C
      ))))

(fn ImpulseVechicle
  (params (Vehicle :type Fort%20Athena%20Vehicle%20Object%20Reference) (Normal :type Vector) (Location :type Vector) (ZMultiplier :type Float%20%28double-precision%29))
  (Audio|BroadcastAbilityImpactatLocation "/Game/Sounds/Creative/Toys/Cube/Cube_LaunchPlayer_Creative_Cue.Cube_LaunchPlayer_Creative_Cue" Vehicle (Math|Vector|vector+vector (* (Math|Vector|MakeVector (.x (* Normal -1.0)) (.y (* Normal -1.0)) (Math|Float|Max(Float) (.z (* Normal -1.0)))) (ScalableFloat|GetValueatLevel (Variables|Default|GetImpulseStrengthVehicle))) (Math|Vector|MakeVector 0.0 0.0 (* ZMultiplier (ScalableFloat|GetValueatLevel (Variables|Default|GetImpulseVehicleZ))))) 7.8e+02) ; Set Physics Linear Velocity
  (bind _output_get (Variables|Default|SetImpulsedVehicle Vehicle))
  (Utilities|Time|SetTimerbyEvent (EventDispatchers|CreateEvent :!event ResetImpulseTimer) 0.5)
  (bind _overlapactor (Variables|Default|GetOverlapActor)
    (:then
      (Class|HyperCubeOverlapActor|CubeBounceFX _overlapactor Location Normal _output_get) ; Fire replicated cube bounce FX in overlap actor
    )
    (:else)))

(fn InitOnSpawn
  ;[1 1 1] Turn on the Cube
  (Variables|Default|SetCubeTurnedOn true))

(fn LaunchPlayerAway
  (params (HitLocation :type Vector) (HitNormal :type Vector) (Player :type Player%20Pawn%20Athena%20Object%20Reference) (ForceMultiplier :type Float%20%28double-precision%29 :default 1.0) (FlashCube :type Boolean))
  (Utilities|FlowControl|Sequence
    (:then_0
      (if FlashCube
        (bind _overlapactor (Variables|Default|GetOverlapActor)
          (:then
            (Class|HyperCubeOverlapActor|CubeBounceFX _overlapactor HitLocation HitNormal Player) ; Fire replicated cube bounce FX in overlap actor
            (CallFunction|PlayerLaunchSound HitLocation))
          (:else))))
    (:then_1
      (Variables|/Script/FortniteGame.FortPawn|SetbIgnoreNextFallingDamage false Player)
      (Parachute|EndSkydiving Player)
      (Character|LaunchCharacter Player (* (Math|Vector|vector+vector (* HitNormal (ScalableFloat|GetValueatLevel (Variables|Default|GetImpulseStrengthPlayer))) (Math|Vector|MakeVector 0.0 0.0 (select (Pawn|Components|CharacterMovement|IsWalking (Variables|/Script/Engine.Character|GetCharacterMovement Player)) (ScalableFloat|GetValueatLevel (Variables|Default|GetImpulsePlayerZWalking)) (ScalableFloat|GetValueatLevel (Variables|Default|GetImpulsePlayerZ))))) ForceMultiplier) true false))))

(fn MakeInterpMovementParameters
  (params (Timeline :type Timeline%20Component%20Object%20Reference))
  (Networking|SwitchHasAuthority
    (:Authority
      (bind (_outlocation _returnvalue) (Helpers|RandomGroundLocationinCircle (Variables|Default|GetSpawnLocation) (Variables|Default|GetMovementRadius) self 2.5e+04 -2.5e+04 (Variables|Default|GetMovementTraceRadius) "(Name=\"FortSupplyDropGroundTrace\")"))
      (if _returnvalue
        (bind _output_get (Variables|Default|SetInterpStartLocation (Transformation|GetActorLocation)))
        (bind _output_get_2 (Variables|Default|SetInterpEndLocation (+ _outlocation (Variables|Default|GetCubeFloatOffset))))
        (bind _distance (Class|MissionBlueprintFunctionLibrary|DistanceBetweenTwoVectors _output_get _output_get_2))
        (bind _output_get_3 (Variables|Default|SetMovementPlayRate (/ (Components|Timeline|GetTimelineLength Timeline) (/ _distance (Variables|Default|GetInterpMoveSpeed)))))
        (return _output_get_3 _output_get _output_get_2)
        (else
          (Development|PrintString "ERROR! BP_HyperCube: Can't find GameMode for 50v50Cube"))))))

(fn MakeInterpMovementPath
  (Variables|Default|SetCurrentControlPointLocation (Transformation|GetActorLocation))
  (for _ (range 11)
    (bind (_outhit _returnvalue) (Collision|SphereTraceByChannel :Start (+ (Variables|Default|GetCurrentControlPointLocation) (Math|Vector|MakeVector (Math|Random|RandomFloatinRange -2e+03 2e+03) (Math|Random|RandomFloatinRange -2e+03 2e+03) 2.5e+04)) :End (+ (Variables|Default|GetCurrentControlPointLocation) (Math|Vector|MakeVector (Math|Random|RandomFloatinRange -2e+03 2e+03) (Math|Random|RandomFloatinRange -2e+03 2e+03) -2.5e+04)) :Radius 1.5e+03 :TraceChannel "TraceTypeQuery10" :bTraceComplex false :DrawDebugType "ForDuration" :bIgnoreSelf true :TraceColor "(R=1.000000,G=0.000000,B=0.000000,A=1.000000)" :TraceHitColor "(R=0.000000,G=1.000000,B=0.000000,A=1.000000)" :DrawTime 0.5))
    (if _returnvalue
      (Variables|Default|SetCurrentControlPointLocation (.Location (Collision|BreakHitResult _outhit)))
      (Utilities|Array|AddUnique (Variables|Default|GetMovePoints) (Utilities|Struct|MakeInterpControlPoint (.Location (Collision|BreakHitResult _outhit)) false))))
  (Variables|/Script/Engine.InterpToMovementComponent|SetControlPoints (Variables|Default|GetMovePoints) (Variables|Default|GetInterpToMovement))
  (Control|RestartMovement (Variables|Default|GetInterpToMovement) 1.0))

(fn OnRep_CubeTurnedOn
  (bind _cubeturnedon (Variables|Default|GetCubeTurnedOn))
  (CallFunction|SetAmbientAudioEnabled _cubeturnedon)
  (if _cubeturnedon
    (Networking|SwitchHasAuthority
      (:Authority
        (Utilities|IsValid (Variables|Default|GetOverlapActor)
          (:Is%20Not%20Valid
            (bind _spawnedactor (Fortnite|Utilities|SpawnTransientActor "/Game/Athena/Playlists/50v50/50v50_Cube/HyperCubeOverlapActor.HyperCubeOverlapActor_C" (Transformation|GetActorTransform) "AlwaysSpawn" self))
            (bind _output_get (Variables|Default|SetOverlapActor _spawnedactor))
            (Transformation|AttachActorToActor _output_get self "None" "SnapToTarget" "KeepWorld" "KeepWorld" true)))))
    (else
      (Networking|SwitchHasAuthority
        (:Authority
          (Utilities|IsValid (Variables|Default|GetOverlapActor)
            (:Is%20Valid
              (Actor|DestroyActor (Variables|Default|GetOverlapActor)))))))))

(fn ShouldDie
  (params (Damage :type Float%20%28single-precision%29) (EventInstigator :type Controller%20Object%20Reference) (DamageCauser :type Actor%20Object%20Reference) (bIsTest :type Boolean) (EffectContext :type Gameplay%20Effect%20Context%20Handle%20Structure%20%28by%20ref%29))
  (return false))

(fn SpawnCoin
  (params (SpawnTransform :type Transform))
  (Networking|SwitchHasAuthority
    (:Authority
      (Game|SpawnActorfromClass "/Game/Athena/Playlists/Score/Coins/BP_ScoreMode_Coin_Gold.BP_ScoreMode_Coin_Gold_C" SpawnTransform "AlwaysSpawn" "OverrideRootScale" 0 0))))

(fn SpawnWeakSpots
  (params (NumWeakSpots :type Integer :default 3))
  (bind _lastusedweakspotsockets (Variables|Default|GetLastUsedWeakSpotSockets))
  (bind _weakspotsockets (Variables|Default|GetWeakSpotSockets))
  (Networking|SwitchHasAuthority
    (:Authority
      (bind _output_get (LocalVariables|SetLocal_ValidWeakSpots _weakspotsockets)) ; make a local copy of the attach socket list
      (if (Utilities|IsArrayNonEmpty _lastusedweakspotsockets)
        (for _array_element _lastusedweakspotsockets
          (Utilities|Array|RemoveItem _output_get _array_element)))
      (Utilities|Array|Shuffle _output_get) ; shuffle the order
      (Utilities|Array|Resize _output_get (Utilities|ClampMaxInteger NumWeakSpots (Utilities|Array|Length _weakspotsockets))) ; resize to the number we want to spawn
      (Variables|Default|SetLastUsedWeakSpotSockets _output_get)
      (for _array_element_2 _output_get
        (bind _returnvalue (Game|SpawnActorfromClass "/Game/Athena/Playlists/50v50/50v50_Cube/HyperCube_WeakSpot.HyperCube_WeakSpot_C" (Math|Transform|MakeTransform (Transformation|GetActorLocation)) "AlwaysSpawn" "OverrideRootScale" self 0 0))
        (Utilities|Array|Add (Variables|Default|GetSpawnedWeakSpotActors) _returnvalue)
        (Transformation|AttachActorToComponent _returnvalue (Variables|Default|GetStaticMesh) _array_element_2 "SnapToTarget" "SnapToTarget" "KeepRelative" true)
        (Development|PrintString (Utilities|String|BuildString(Object) "" "BP_HyperCube: Spawn WeakSpot!  
Object: " _returnvalue) (Variables|Default|GetDebugPrints) (Variables|Default|GetDebugPrints) "(R=0.585798,G=0.000000,B=1.000000,A=1.000000)")))))

(fn ConstructionScript)
