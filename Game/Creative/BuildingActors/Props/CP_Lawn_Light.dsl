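; CP_Lawn_Light is a Creative lawn light prop Blueprint with a spot light and a particle light ray.
; Its construction script pushes the LightColorTint variable into the ray particle's LightColor parameter and onto the spot light's color.
; The LightOpacity variable is remapped by a lerp into a low opacity range for the ray's LightOpacity parameter, so it does not affect the spot light.
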(fn ConstructionScript
  (bind _p_lawnlight_ray (Variables|Default|GetP_LawnLight_Ray))
  (Effects|Components|ParticleSystem|SetVectorParameter _p_lawnlight_ray "LightColor" (Math|Conversions|ToVector(LinearColor) (Variables|Default|GetLightColorTint)))
  (Effects|Components|ParticleSystem|SetFloatParameter _p_lawnlight_ray "LightOpacity" (Math|Float|Lerp 0.0 0.3 (Variables|Default|GetLightOpacity)))
  (Rendering|Components|Light|SetLightColor (Variables|Default|GetSpotLight) (Variables|Default|GetLightColorTint) true))
