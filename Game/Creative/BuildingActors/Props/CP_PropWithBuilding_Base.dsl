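; CP_PropWithBuilding_Base is a Creative prop base Blueprint that lets UEFN reference assets used by its static mesh.
; It extends the parent's exposed asset provider hooks, registering version 2 after the parent's entries.
; Its collection override calls the parent, then adds every mesh material and the overlay material as indirect references, intended to allow their texture data.
; The construction script is empty.
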(fn BuildExposedAssetProviderVersion(Implementation)
  (params (Version :type Fort%20Exposed%20Asset%20Provider%20Version%20Structure%20%28by%20ref%29))
  (|CallParentFunction Version)
  (ExposedAssetProvider|AddVersion Version 2)
  (return))

(fn CollectAssetReferencestoAllow(Implementation)
  (params (Collector :type Fort%20Exposed%20Asset%20Reference%20Collector%20Structure%20%28by%20ref%29))
  (bind _staticmeshcomponent (Variables|Debug|GetStaticMeshComponent))
  (|CallParentFunction Collector)
  (for _array_element (Rendering|Material|GetMaterials _staticmeshcomponent)
    (ExposedAssetProvider|CollectIndirectAssetReference Collector (Utilities|ResolveSoftReference _array_element)) ; Collect the texture data to allow it to be referenced in UEFN
  )
  (ExposedAssetProvider|CollectIndirectAssetReference Collector (Utilities|ResolveSoftReference (Variables|/Script/Engine.MeshComponent|GetOverlayMaterial _staticmeshcomponent))) ; Collect the texture data to allow it to be referenced in UEFN
  (return))

(fn ConstructionScript)
