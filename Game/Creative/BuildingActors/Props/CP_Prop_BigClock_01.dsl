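; Creative big clock prop Blueprint whose hour and minute hands follow the in-game time of day.
; On BeginPlay, only off mobile and where cosmetic events can execute, it starts a looping one-second timer calling UpdateTime.
; UpdateTime sets each hand mesh's relative rotation from GetTimeOfDay, using a twelve-hour cycle for the hour hand and whole-and-remainder division by one hour for the minute hand.
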(event EventBeginPlay
  (if (not (Fortnite|IsMobilePlatform))
    (Networking|CanExecuteCosmeticEvents
      (:True
        (Utilities|Time|SetTimerbyFunctionName 0 "UpdateTime" 1 true)))))

(fn UpdateTime
  (Transformation|SetRelativeRotation (Variables|BigClock_01|GetHourHandMesh) (Math|Rotator|MakeRotator 0.0 (- 360 (* (/ (TimeOfDay|GetTimeOfDay :!class /Script/FortniteGame.FortKismetLibrary) 12) 360))) false false)
  (Transformation|SetRelativeRotation (Variables|BigClock_01|GetMinuteHandMesh) (Math|Rotator|MakeRotator 0.0 (- 360 (* (Math|Float|Division(WholeandRemainder) (TimeOfDay|GetTimeOfDay :!class /Script/FortniteGame.FortKismetLibrary) 1) 360))) false false))

(fn ConstructionScript)
