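; Creative prop Blueprint for the Demeter lighthouse light that clears the lighthouse's spinning beam effects when it is destroyed.
; On destruction, the server gathers every actor of the spinny light beams prototype class in the world and destroys each one.
; The lookup is by class rather than by reference, so it removes all such beam actors, including those of other lighthouses.
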(event EventDestroyed
  (if (Networking|IsServer)
    (bind _outactors (Actor|GetAllActorsOfClass "/Game/Effects/Fort_Effects/World/Enviroment/Lighthouse/BP_Demeter_Lighthouse_Prototype_Spinny_LightBeams.BP_Demeter_Lighthouse_Prototype_Spinny_LightBeams_C"))
    (for _array_element _outactors
      (Actor|DestroyActor _array_element))))

(fn ConstructionScript)
