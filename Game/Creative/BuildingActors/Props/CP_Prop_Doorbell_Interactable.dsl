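; Interactable Creative doorbell prop that rings by house type, with glow and scale effects, and breaks temporarily when rung too often.
; Interaction increments DoorbellRungNum, reset after a short idle timer; its OnRep plays client effects or disables interaction on the server until DownTime ends.
; Outside creative mode and simple-interact quests, ringing box-checks the house for enemy-team pawns to send a quest stat, while local quest updates trigger success effects.
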(event EventBeginPlay
  (bind _returnvalue (Fortnite|GetLocalFortPlayerControllers true))
  (bind _output (Utilities|Array|Get(acopy) _returnvalue))
  ;[1 1 1] Random value for holiday house type
  (if (Variables|Default|GetbIsHoliday)
    (Variables|Default|SetHouseType "NewEnumerator3")
    (else
      _))
  ;[1 1 1] Check if In Creative Mode or not
  (bind _asfort_game_state_athena (Utilities|Casting|CastToFortGameStateAthena (Game|GetGameState)))
  (if (Variables|/Script/FortniteGame.FortGameStateAthena|GetbPlaylistDataIsLoaded _asfort_game_state_athena)
    (bind _cancompletequest (Class|AthenaQuestFunctionLibrary|AthenaCheckCreativeMode _asfort_game_state_athena))
    (if _cancompletequest
      (Variables|QuestStuff|SetCreativeMode false)
      (else
        (Variables|QuestStuff|SetCreativeMode true)))
    ;[1 1 1] Set up DMI
    (Networking|SwitchHasAuthority
      (:Remote
        (bind _returnvalue_2 (Rendering|Material|CreateDynamicMaterialInstance (Variables|Default|GetDoorbellVisMesh) 0 "/Game/Environments/Props/Residential/Materials/MI_Door_Bell_01.MI_Door_Bell_01" "None" :!class /Script/Engine.PrimitiveComponent))
        (Variables|Default|SetDoorbellDMI _returnvalue_2)
        ;[1 0.587 0] HAAAACKS
        (Utilities|FlowControl|Delay 5.0)
        ;[0.5587 0.217 1] I'm going to use this boolean and block any part that doesn't might interfere with a basic interact quest setup like what we do for chests for example.
        (if (not (Variables|Default|GetbIsSimpleInteractQuest))
          ;[1 1 1] Bindings for Quest Updating
          (bind (_questisvalid _questcompleted _objectivecompleted) (Default|AthenaCheckQuestandObjectiveCompletion _output (Variables|QuestStuff|GetQuestItem) (Variables|QuestStuff|GetObjBackendName)))
          (if _questisvalid
            (if _objectivecompleted
              _
              (elif (Variables|QuestStuff|GetCreativeMode)
                _
                (else
                  (Quests|BindEventtoOnQuestsUpdated (Quests|GetQuestManager _output "Invalid") (EventDispatchers|CreateEvent :!event Custom|QuestUpdated))
                  (Variables|QuestStuff|SetLocallyActive true)
                  (bind _questcountachieved (Class|AthenaQuestFunctionLibrary|AthenaQuestObjectiveCounter _output (Variables|QuestStuff|GetQuestItem)))
                  (bind _output_get (Variables|QuestStuff|SetCurrentObjCount _questcountachieved))
                  (Development|PrintString (Utilities|String|ToString(Integer) _output_get)))))))))
    (else
      (Playlist|BindEventtoOnCurrentPlaylistLoaded _asfort_game_state_athena (delegate Custom|PlaylistLoaded)))))

(event Custom|PlaylistLoaded
  (params (PlaylistName :type Name) (PlaylistContextTags :type Gameplay%20Tag%20Container%20Structure%20%28by%20ref%29))
  (bind _returnvalue (Fortnite|GetLocalFortPlayerControllers true))
  (bind _output (Utilities|Array|Get(acopy) _returnvalue))
  ;[1 1 1] Check if In Creative Mode or not
  (bind _cancompletequest (Class|AthenaQuestFunctionLibrary|AthenaCheckCreativeMode (Game|GetGameState)))
  (if _cancompletequest
    (Variables|QuestStuff|SetCreativeMode false)
    (else
      (Variables|QuestStuff|SetCreativeMode true)))
  ;[1 1 1] Set up DMI
  (Networking|SwitchHasAuthority
    (:Remote
      (bind _returnvalue_2 (Rendering|Material|CreateDynamicMaterialInstance (Variables|Default|GetDoorbellVisMesh) 0 "/Game/Environments/Props/Residential/Materials/MI_Door_Bell_01.MI_Door_Bell_01" "None" :!class /Script/Engine.PrimitiveComponent))
      (Variables|Default|SetDoorbellDMI _returnvalue_2)
      ;[1 0.587 0] HAAAACKS
      (Utilities|FlowControl|Delay 5.0)
      ;[0.5587 0.217 1] I'm going to use this boolean and block any part that doesn't might interfere with a basic interact quest setup like what we do for chests for example.
      (if (not (Variables|Default|GetbIsSimpleInteractQuest))
        ;[1 1 1] Bindings for Quest Updating
        (bind (_questisvalid _questcompleted _objectivecompleted) (Default|AthenaCheckQuestandObjectiveCompletion _output (Variables|QuestStuff|GetQuestItem) (Variables|QuestStuff|GetObjBackendName)))
        (if _questisvalid
          (if _objectivecompleted
            _
            (elif (Variables|QuestStuff|GetCreativeMode)
              _
              (else
                (Quests|BindEventtoOnQuestsUpdated (Quests|GetQuestManager _output "Invalid") (EventDispatchers|CreateEvent :!event Custom|QuestUpdated))
                (Variables|QuestStuff|SetLocallyActive true)
                (bind _questcountachieved (Class|AthenaQuestFunctionLibrary|AthenaQuestObjectiveCounter _output (Variables|QuestStuff|GetQuestItem)))
                (bind _output_get (Variables|QuestStuff|SetCurrentObjCount _questcountachieved))
                (Development|PrintString (Utilities|String|ToString(Integer) _output_get))))))))))

(event Custom|QuestUpdated
  (bind _returnvalue (Fortnite|GetLocalFortPlayerControllers true))
  (bind _output (Utilities|Array|Get(acopy) _returnvalue))
  ;[1 1 1] Checking for Quest Success
  (if (Variables|QuestStuff|GetLocallyActive)
    (bind (_questisvalid _questcompleted) (Default|AthenaCheckQuestandObjectiveCompletion _output (Variables|QuestStuff|GetQuestItem) (Variables|QuestStuff|GetObjBackendName)))
    (if (not _questcompleted)
      (bind _questcountachieved (Class|AthenaQuestFunctionLibrary|AthenaQuestObjectiveCounter _output (Variables|QuestStuff|GetQuestItem)))
      (if (> _questcountachieved (Variables|QuestStuff|GetCurrentObjCount))
        _))
    (Variables|QuestStuff|SetLocallyActive false)
    (CallFunction|PlayQuestSuccessFX)
    ;[1 1 1] Bindings for Quest Updating
    (Quests|UnbindEventfromOnQuestsUpdated (Quests|GetQuestManager (Utilities|Array|Get(acopy) (Fortnite|GetLocalFortPlayerControllers true)) "Invalid") (EventDispatchers|CreateEvent :!event Custom|QuestUpdated))))

(event Custom|PlayQuestSuccessFX
  (bind _returnvalue (Fortnite|GetLocalFortPlayerControllers true))
  (bind _output (Utilities|Array|Get(acopy) _returnvalue))
  (bind _returnvalue_2 (Game|Player|GetPlayerPawn _output))
  ;[1 1 1] Quest Success FX
  (Development|PrintString "SUCCESS FX" true false "(R=1.000000,G=0.703857,B=0.040648,A=1.000000)" 5.0)
  (Utilities|FlowControl|Delay 1.0)
  (Effects|SpawnFXSystemAttached (Variables|QuestStuff|GetQuestSuccessEmitter) (Variables|/Script/Engine.Character|GetMesh _returnvalue_2))
  (Audio|PlaySoundatLocation (Variables|QuestStuff|GetQuestSuccessSound) (Transformation|GetActorLocation _returnvalue_2)) ; Quest Success Sound
)

(event Interaction|EventBlueprintOnInteract
  (params (InteractingPawn :type Fort%20Pawn%20Object%20Reference) (InteractionBeingAttempted :type EInteractionBeingAttempted%20Enum :default "FirstInteraction"))
  ;[1 1 1] Check for Enemies in house when you ring da bell
  (Variables|Default|SetPawnWhoRangDoorbell InteractingPawn)
  (Utilities|Time|ClearandInvalidateTimerbyHandle (Variables|Default|GetBreakdownTimer))
  (Variables|Default|SetDoorbellRungNum (+ (Variables|Default|GetDoorbellRungNum) 1))
  (bind _returnvalue (Utilities|Time|SetTimerbyEvent (delegate Custom|ResetBreakTimer) 5.0))
  (Variables|Default|SetBreakdownTimer _returnvalue)
  (if (Variables|QuestStuff|GetCreativeMode)
    _
    (elif (Variables|Default|GetbIsSimpleInteractQuest)
      _
      (else
        (Variables|QuestStuff|SetEnemyInHouse false)
        (bind _outactors (Collision|BoxOverlapActors (Transformation|GetWorldLocation (Variables|Default|GetInsideHouseOverlap)) (Components|Box|GetUnscaledBoxExtent (Variables|Default|GetInsideHouseOverlap)) (Variables|Default|GetObjectTypes) "/Script/FortniteGame.FortPawn" :!class /Script/Engine.KismetSystemLibrary))
        (bind _array_element (Utilities|Array|ForEachLoopwithBreak _outactors)
          (:LoopBody
            (if (Math|Byte|NotEqual(Byte) (Team|GetActorTeam _array_element) (Team|GetActorTeam (Variables|Default|GetPawnWhoRangDoorbell)))
              (Variables|QuestStuff|SetEnemyInHouse true)))
          (:Completed
            _))
        ;[1 1 1] Broadcast Quest Progress
        (if (Variables|QuestStuff|GetEnemyInHouse)
          (Quest|SendComplexCustomStatEvent (Quests|GetQuestManager (Fortnite|GetFortPlayerControllerfromActor (Variables|Default|GetPawnWhoRangDoorbell)) "Invalid") 0 "(GameplayTags=((TagName=\"Athena.Quests.Doorbell.EnemyInside\")))" "(GameplayTags=)" 1)
          (Development|PrintString "Enemy In House" false false "(R=0.234243,G=1.000000,B=0.000000,A=1.000000)" 5.0)
          (else
            (Development|PrintString "Enemy NOT  In House" false false "(R=1.000000,G=0.000000,B=0.051918,A=1.000000)" 5.0)))))))

(event Custom|ResetBreakTimer
  ;[1 1 1] Check for Enemies in house when you ring da bell
  (Variables|Default|SetDoorbellRungNum 0))

(event Custom|DoorBell%20Rung%20Event
  ;[1 1 1] Doorbell FX and Sound
  (CallFunction|PlayDoorBellSound)
  (bind (_direction _glow _scale) (Timeline|DoorbellRungFX PlayFromStart)
    (:Update
      (Rendering|Material|SetVectorParameterValue (Variables|Default|GetDoorbellDMI) "Emissive Multiplier" (Math|Color|Lerp(LinearColor) "(R=1.000000,G=1.000000,B=1.000000,A=1.000000)" "(R=10.000000,G=10.000000,B=10.000000,A=10.000000)" _glow) :!class /Script/Engine.MaterialInstanceDynamic)
      (Transformation|SetRelativeScale3D (Variables|Default|GetDoorbellVisMesh) (Math|Vector|Lerp(Vector) "1.000000,1.000000,1.000000" "2.500000,1.300000,1.300000" _scale)))))

(event Custom|StopRingFX
  (bind _doorbelldmi (Variables|Default|GetDoorbellDMI))
  (Utilities|FlowControl|Sequence
    (:then_0
      ;[1 1 1] Doorbell FX and Sound
      (Timeline|DoorbellRungFX Stop))
    (:then_1
      ;[1 0.217 0.9601] @Ewing Changed this to match the vis mesh scale
      (Transformation|SetRelativeScale3D (Variables|Default|GetDoorbellVisMesh) "1.000000,1.000000,1.000000")
      ;[1 1 1] Doorbell Flickers when it goes out
      (bind (_direction _glow) (Timeline|Timeline_0 PlayFromStart)
        (:Update
          (Rendering|Material|SetVectorParameterValue _doorbelldmi "Emissive Multiplier" (Math|Color|Lerp(LinearColor) "(R=0.000000,G=0.000000,B=0.000000,A=1.000000)" "(R=10.000000,G=10.000000,B=10.000000,A=10.000000)" _glow) :!class /Script/Engine.MaterialInstanceDynamic))
        (:Finished
          (Rendering|Material|SetVectorParameterValue _doorbelldmi "Emissive Multiplier" "(R=0.000000,G=0.000000,B=0.000000,A=1.000000)" :!class /Script/Engine.MaterialInstanceDynamic))))))

(event Custom|BrokenDoorbellEvent
  ;[1 1 1] Play broken doorbell Sound and FX
  (Audio|PlaySoundatLocation (Variables|Default|GetBrokenDoorbellSound) (Transformation|GetWorldLocation (Variables|Default|GetDoorbellButtonLocation)))
  (Effects|SpawnEmitteratLocation (Variables|Default|GetBreakdownFX) (Transformation|GetWorldLocation (Variables|Default|GetDoorbellVisMesh)) (Transformation|GetWorldRotation (Variables|Default|GetDoorbellVisMesh)) "0.500000,0.500000,0.500000")
  (CallFunction|PlayDoorBellSound)
  (bind (_direction _pitch _volume) (Timeline|Bell%20Pitch%20Down PlayFromStart)
    (:Update
      (bind _currentlyplayingbellsound (Variables|Default|GetCurrentlyPlayingBellSound)
        (:then
          (Audio|Components|Audio|SetPitchMultiplier _currentlyplayingbellsound _pitch)
          (Audio|Components|Audio|SetVolumeMultiplier _currentlyplayingbellsound _volume :!class /Script/Engine.AudioComponent))
        (:else)))))

(event Custom|DoorStartsToWorkAgain
  ;[1 1 1] Doorbell starts to work again
  (Variables|Default|SetDoorbellRungNum 0)
  (Variables|Default|SetCanInteract true)
  (Networking|ForceNetUpdate))

(fn BlueprintCanInteract
  (params (InteractingPawn :type Fort%20Pawn%20Object%20Reference) (InteractionBeingAttempted :type EInteractionBeingAttempted%20Enum :default "FirstInteraction") (InteractionType :type TInteractionType%20Enum :default "IT_NoInteraction"))
  ;[1 1 1] Check if DMI still proper as it can be replaced after damage
  (if (Utilities|NotEqual(Object) (Rendering|Material|GetMaterial (Variables|Default|GetDoorbellVisMesh) 0) (Variables|Default|GetDoorbellDMI))
    (Rendering|Material|SetMaterial (Variables|Default|GetDoorbellVisMesh) 0 (Variables|Default|GetDoorbellDMI))
    (else
      _))
  (return (Variables|Default|GetCanInteract)))

(fn BlueprintGetInteractionString
  (params (InteractingPawn :type Fort%20Pawn%20Object%20Reference) (InteractionBeingAttempted :type EInteractionBeingAttempted%20Enum :default "FirstInteraction"))
  (return (Variables|Default|GetRing%20Bell%20String)))

(fn OnRep_CanInteract
  (bind _doorbelldmi (Variables|Default|GetDoorbellDMI))
  (if (Variables|Default|GetCanInteract)
    (Rendering|Material|SetVectorParameterValue _doorbelldmi "Emissive Multiplier" "(R=1.000000,G=1.000000,B=1.000000,A=1.000000)" :!class /Script/Engine.MaterialInstanceDynamic)
    (else
      (Rendering|Material|SetVectorParameterValue _doorbelldmi "Emissive Multiplier" "(R=0.000000,G=0.000000,B=0.000000,A=1.000000)" :!class /Script/Engine.MaterialInstanceDynamic))))

(fn OnRep_DoorbellRungNum
  (if (not (== (Variables|Default|GetDoorbellRungNum) 0))
    ;[1 1 1] added to override network update throttle to allow the doorbell to be rung quickly before it 'breaks' and gets shut off
    (Networking|ForceNetUpdate)
    ;[1 1 1] Check if DMI still proper as it can be replaced after damage
    (if (Utilities|NotEqual(Object) (Rendering|Material|GetMaterial (Variables|Default|GetDoorbellVisMesh) 0) (Variables|Default|GetDoorbellDMI))
      (Rendering|Material|SetMaterial (Variables|Default|GetDoorbellVisMesh) 0 (Variables|Default|GetDoorbellDMI))
      (else
        _))
    (if (>= (Variables|Default|GetDoorbellRungNum) (Variables|Default|GetBreakdownNumber))
      ;[1 1 1] Doorbell broken from ringning too much
      (Networking|SwitchHasAuthority
        (:Authority
          (Utilities|Time|SetTimerbyEvent (EventDispatchers|CreateEvent :!event DoorStartsToWorkAgain) (Variables|Default|GetDownTime))
          (Variables|Default|SetCanInteract false))
        (:Remote
          (CallFunction|StopRingFX)
          (CallFunction|BrokenDoorbellEvent)))
      (else
        ;[1 1 1] Play Doorbell sound and FX
        (Networking|SwitchHasAuthority
          (:Remote
            (CallFunction|DoorBellRungEvent)))))))

(fn PlayDoorBellSound
  (Utilities|FlowControl|Sequence
    (:then_0
      (bind _currentlyplayingbellsound (Variables|Default|GetCurrentlyPlayingBellSound)
        (:then
          (Audio|Components|Audio|FadeOut _currentlyplayingbellsound 0.1 0.0 "Linear" :!class /Script/Engine.AudioComponent))
        (:else
          _))
      (bind _returnvalue (Audio|SpawnSoundatLocation (Utilities|Select :Index (Variables|Default|GetHouseType) :NewEnumerator0 (Variables|Default|GetFancyDoorbellSound) :NewEnumerator1 (Variables|Default|GetRegularDoorbellSound) :NewEnumerator2 (Variables|Default|GetRundownDoorbellSound) :NewEnumerator3 (Variables|Default|GetHolidayDoorbellSound)) (Transformation|GetWorldLocation (Variables|Default|GetDoorbellLocation))))
      (Variables|Default|SetCurrentlyPlayingBellSound _returnvalue))
    (:then_1
      (Audio|PlaySoundatLocation (Variables|Default|GetDoorbellPressSound) (Transformation|GetWorldLocation (Variables|Default|GetDoorbellButtonLocation))))))

(fn ConstructionScript)
