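; Creative island player spawn pad prop whose interaction toggles the visibility of its hub restriction volume and flips the prompt between Zone On and Zone Off.
; The replicated, saved visibility flag's OnRep briefly disables interaction, refreshes visuals and resaves.
; The interacting client disables its own interaction first, using a retriggerable delay, to hide replication lag.
; On BeginPlay it applies the saved state and destroys itself in a curated hub.
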(event EventBeginPlay
  (bind _self self)
  (bind _returnvalue (Minigame|GetMinigame _self :!class /Script/FortniteGame.MinigameBlueprintLibrary))
  (Utilities|FlowControl|Sequence
    (:then_0
      (CallFunction|UpdateVisibility) ; Set initial visibility State (b_IsVisible is a save game variable)
    )
    (:then_1
      ;[1 1 1] Destroy this actor if it's loaded in the actual creative HUB
      (Utilities|IsValid _returnvalue
        (:Is%20Valid
          (if (Minigame|IsCuratedHub _returnvalue)
            (Actor|DestroyActor)))))))

(event Interaction|EventBlueprintOnInteract
  (params (InteractingPawn :type Fort%20Pawn%20Object%20Reference) (InteractionBeingAttempted :type EInteractionBeingAttempted%20Enum :default "FirstInteraction"))
  ;[1 1 1] When Interact is called, reverse IsVisible
  (Variables|Default|Setb_IsVisible (not (Variables|Default|Getb_IsVisible))))

(event Custom|Disable%20Interact%20for%20Duration
  (Variables|Default|Setb_CanInteract false)
  (Utilities|FlowControl|RetriggerableDelay (Variables|Default|GetRetrigger%20Delay))
  (Variables|Default|Setb_CanInteract true))

(fn BlueprintCanInteract
  (params (InteractingPawn :type Fort%20Pawn%20Object%20Reference) (InteractionBeingAttempted :type EInteractionBeingAttempted%20Enum :default "FirstInteraction") (InteractionType :type TInteractionType%20Enum :default "IT_NoInteraction"))
  (return (Variables|Default|Getb_CanInteract)))

(fn BlueprintGetInteractionString
  (params (InteractingPawn :type Fort%20Pawn%20Object%20Reference) (InteractionBeingAttempted :type EInteractionBeingAttempted%20Enum :default "FirstInteraction"))
  (return (Variables|Default|GetInteractText)))

(fn BlueprintOnLocalInteract
  (params (InteractingPawn :type Fort%20Player%20Pawn%20Object%20Reference) (InteractionBeingAttempted :type EInteractionBeingAttempted%20Enum :default "FirstInteraction"))
  (CallFunction|DisableInteractForDuration) ; Disabling interact when the client interacts to avoid a visual glitch\nwhere you can keep interacting for a short duration after interacting until\nthe server replicates the Visibility boolean. This means that the client will disable its\nown interaction first.
  (return false))

(fn OnRep_b_IsVisible
  (CallFunction|DisableInteractForDuration)
  (CallFunction|UpdateVisibility)
  (CallFunction|Resave))

(fn UpdateVisibility
  (Networking|CanExecuteCosmeticEvents
    (:True
      (Rendering|SetVisibility (Variables|Default|GetHub_Restriction_Volume) (Variables|Default|Getb_IsVisible) false)
      (if (Variables|Default|Getb_IsVisible)
        (Variables|Default|SetInteractText "NSLOCTEXT(\"[2D2859574B8349D2C33E91ADD3704FE8]\", \"EAD6513843DB3E2A22348EAEEFEE8EBF\", \"Zone Off\")")
        (else
          (Variables|Default|SetInteractText "NSLOCTEXT(\"[2D2859574B8349D2C33E91ADD3704FE8]\", \"9175D1144C26860D266B398640520FB2\", \"Zone On\")"))))))

(fn ConstructionScript
  (|CallParentFunction))
