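; Creative island portal prop that removes itself when it is loaded in a curated Creative hub.
; On BeginPlay it fetches the current Minigame, and if that is valid and flagged as a curated hub, it destroys the actor.
; The construction script only calls its parent.
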(event EventBeginPlay
  (bind _self self)
  (bind _returnvalue (Minigame|GetMinigame _self :!class /Script/FortniteGame.MinigameBlueprintLibrary))
  ;[1 1 1] Destroy this actor if it's loaded in the actual creative HUB
  (Utilities|IsValid _returnvalue
    (:Is%20Valid
      (if (Minigame|IsCuratedHub _returnvalue)
        (Actor|DestroyActor)))))

(fn ConstructionScript
  (|CallParentFunction))
