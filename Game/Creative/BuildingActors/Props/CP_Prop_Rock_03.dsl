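; Creative-mode prop actor Blueprint for a rock (CP_Prop_Rock_03) that decides whether it can be interacted with in Creative.
; Its IsInteractionAllowedInCreative function returns the Creative move-tool interaction setting, ignoring the passed interaction behavior.
; The construction script is empty.
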(fn IsInteractionAllowedInCreative
  (params (InteractionBehavior :type Object%20Interaction%20Behavior%20Object%20Reference))
  (return (Variables|Creative|GetIsInteractionWithMoveToolAllowed)))

(fn ConstructionScript)
