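; Creative-mode small SilkyTrombone prop Blueprint that plays its own explosion effect when destroyed.
; On death it calls the parent effects, then spawns the small SilkyTrombone explosion emitter at the static mesh's DeathFX socket location.
; The spawned emitter is set to absolute scale so the effect keeps its correct size.
; BeginPlay, overlap, tick and construction add no logic of their own.
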(event EventBeginPlay
  (|CallParentFunction) ; This node is disabled and will not be called.\nDrag off pins to build functionality.
)

(event Collision|EventActorBeginOverlap
  (params (OtherActor :type Actor%20Object%20Reference))
  (|CallParentFunction OtherActor) ; This node is disabled and will not be called.\nDrag off pins to build functionality.
)

(event EventTick
  (params (DeltaSeconds :type Float%20%28single-precision%29))
  (|CallParentFunction DeltaSeconds) ; This node is disabled and will not be called.\nDrag off pins to build functionality.
)

(event EventOnDeathPlayEffects
  (params (Damage :type Float%20%28single-precision%29) (DamageTags :type Gameplay%20Tag%20Container%20Structure%20%28by%20ref%29) (Momentum :type Vector :default "0, 0, 0") (HitInfo :type Hit%20Result%20Structure%20%28by%20ref%29) (InstigatedBy :type Fort%20Pawn%20Object%20Reference) (DamageCauser :type Actor%20Object%20Reference) (EffectContext :type Gameplay%20Effect%20Context%20Handle%20Structure))
  ;[1 1 1] Keeps Death Effect in correct Scale
  (|CallParentFunction Damage DamageTags Momentum HitInfo InstigatedBy DamageCauser EffectContext)
  (bind _returnvalue (Effects|SpawnEmitteratLocation "/Game/Athena/Items/Gameplay/SilkyTrombone/FX/P_SilkyTrombone_Explode_Small.P_SilkyTrombone_Explode_Small" (Transformation|GetSocketLocation (Variables|Debug|GetStaticMeshComponent) "DeathFX")))
  (Variables|/Script/Engine.SceneComponent|SetbAbsoluteScale true _returnvalue))

(fn ConstructionScript
  (|CallParentFunction))
