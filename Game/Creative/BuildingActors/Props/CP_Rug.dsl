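; CP_Rug is a Creative building prop Blueprint for a rug that removes itself when an actor it is attached to is destroyed.
; On BeginPlay it loops over the actors from its RugAttachmentActors array and binds DestroyedOuch to each valid one's OnDestroyed event.
; DestroyedOuch destroys the rug when the reported actor passes an IsValid check, so losing any single attachment is enough.
; The construction script is empty.
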(event EventBeginPlay
  (if (> (Utilities|Array|Length (Variables|Default|GetRugAttachmentActors)) 0)
    (for _array_element (Variables|Default|GetRugAttachmentActors)
      (Utilities|IsValid _array_element
        (:Is%20Valid
          (Game|BindEventtoOnDestroyed _array_element (delegate Custom|DestroyedOuch)))))))

(event Custom|DestroyedOuch
  (params (DestroyedActor :type Actor%20Object%20Reference))
  (Utilities|IsValid DestroyedActor
    (:Is%20Valid
      (Actor|DestroyActor self))))

(fn ConstructionScript)
