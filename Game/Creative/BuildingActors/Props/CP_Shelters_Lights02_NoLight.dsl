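; Creative building prop Blueprint for a shelter light fixture, the NoLight variant going by its name, with an optional moth bug effect.
; The construction script runs the parent's logic, then checks the EnableBugEffect flag.
; When enabled, it adds a static mesh component at MothRelativeOffset and sets that component's world scale from the Moth Scale variable.
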(fn ConstructionScript
  (|CallParentFunction)
  (if (Variables|Default|GetEnableBugEffect)
    (bind _returnvalue (AddComponent|NODE_AddStaticMeshComponent-0 :RelativeTransform (Variables|Default|GetMothRelativeOffset)))
    (Transformation|SetWorldScale3D _returnvalue (Variables|Default|GetMoth%20Scale))))
