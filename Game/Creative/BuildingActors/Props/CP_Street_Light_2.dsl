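; Creative street light prop Blueprint with a spot light and a light ray that responds to day phase changes.
; After calling the parent handler, the day phase event switches on the new phase to toggle light ray visibility, with an explicit show for Evening.
; The construction script disables spot light shadow casting on Nintendo Switch.
; A dynamic destruction override supplies the static mesh component in all three returned arrays.
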(event EventBeginPlay)

(event EventOnDayPhaseChanged
  (params (CurrentDayPhase :type EFortDayPhase%20Enum :default "Morning") (PreviousDayPhase :type EFortDayPhase%20Enum :default "Morning") (bAtCreation :type Boolean))
  (bind _lightray (Variables|Default|GetLightRay))
  (|CallParentFunction CurrentDayPhase PreviousDayPhase bAtCreation)
  (switch Utilities|FlowControl|Switch|SwitchonEFortDayPhase CurrentDayPhase
    (:Morning
      _)
    (:Day
      _)
    (:Evening
      (Rendering|SetVisibility _lightray true false))
    (:Night
      _))
  (Rendering|SetVisibility _lightray false false))

(fn GetDynamicDestructionStaticMeshComponents
  (return (Utilities|Array|MakeArray (Variables|Debug|GetStaticMeshComponent)) (Utilities|Array|MakeArray (Variables|Debug|GetStaticMeshComponent)) (Utilities|Array|MakeArray (Variables|Debug|GetStaticMeshComponent))))

(fn ConstructionScript
  (|CallParentFunction)
  (if (TouchPlatform|IsNintendoSwitch)
    (Rendering|Components|Light|SetCastShadows (Variables|Default|GetSpotLight) false)))
