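; Creative hanging lamp prop Blueprint whose construction script applies optional light overrides to its point light.
; It calls the parent construction script, then checks three toggle variables in the Light category.
; Each enabled toggle pushes its paired variable onto the point light: shadow casting, attenuation radius or intensity.
; A disabled toggle leaves that light property untouched.
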(fn ConstructionScript
  (bind _pointlight (Variables|Default|GetPointLight))
  (|CallParentFunction)
  (if (Variables|Light|GetSet%20Cast%20Shadow)
    (Rendering|Components|Light|SetCastShadows _pointlight (Variables|Light|GetCast%20Shadows))
    (else
      _))
  (if (Variables|Light|GetChange%20Attenuation%20Radius)
    (Rendering|Lighting|SetAttenuationRadius _pointlight (Variables|Light|GetAttenuation%20Radius))
    (else
      _))
  (if (Variables|Light|GetChange%20Intensity)
    (Rendering|Components|Light|SetIntensity _pointlight (Variables|Light|GetIntensity))))
