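; Parent Blueprint for Creative grass props whose color is set through user options, falling back to the mesh's default color when customization is off.
; On begin play and toy option loads or changes, cosmetic-capable instances cache the default material, MID and color, then update visuals.
; The chosen color goes to the cached MID, plus the mesh's current material when it differs, covering materials applied by the phone tool.
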(event OnPropertiesLoaded(ToyOptionsComponent)
  (Networking|CanExecuteCosmeticEvents
    (:True
      (CallFunction|CacheAllDefaultMaterials)
      (CallFunction|UpdateVisualsonClient))))

(event OnAnyPropertyChanged(ToyOptionsComponent)
  (Networking|CanExecuteCosmeticEvents
    (:True
      (CallFunction|CacheAllDefaultMaterials)
      (CallFunction|UpdateVisualsonClient))))

(event EventBeginPlay
  (Networking|CanExecuteCosmeticEvents
    (:True
      (CallFunction|CacheAllDefaultMaterials)
      (CallFunction|UpdateVisualsonClient))))

(fn CacheAllDefaultMaterials
  (bind _newlycached (Default|CacheDefaultMaterialandMID (Variables|Default|GetCreative_Grass) (Variables|Materials|GetGrassMatIndex) (Variables|Materials|GetDefaultGrassMaterial) (Variables|Materials|MID|GetGrass_MID)))
  (if _newlycached
    (Default|CacheDefaultColor (Variables|Materials|MID|GetGrass_MID) (Variables|Parameters|GetGrassColor_Param) (Variables|Parameters|GetGrassColor_Default))))

(fn CacheDefaultMaterialAndMID
  (params (Target%20Mesh :type Static%20Mesh%20Component%20Object%20Reference) (Target%20Material%20Index :type Integer) (Material%20Variable :type Material%20Interface%20Object%20Reference%20%28by%20ref%29) (MID%20Variable :type Material%20Instance%20Dynamic%20Object%20Reference%20%28by%20ref%29))
  (bind _target_material_index (Variables|GetTarget%20Material%20Index))
  (bind _target_mesh (Variables|GetTarget%20Mesh))
  (bind _returnvalue (Rendering|Material|GetMaterial _target_mesh _target_material_index))
  (bind _material_variable (Variables|GetMaterial%20Variable))
  (bind _target_material_index_2 (Variables|GetTarget%20Material%20Index))
  (bind _target_mesh_2 (Variables|GetTarget%20Mesh))
  (bind _mid (Material|GetMIDfromMaterial _target_mesh_2 _target_material_index_2 _material_variable))
  ;[1 0.7813 0.01256 0.5] Check and cache default material
  (Utilities|IsValid (Variables|GetMaterial%20Variable)
    (:Is%20Valid
      ;[0.7537 1 0.06522 0.5] Set the MID if needed
      (Utilities|IsValid (Variables|GetMID%20Variable)
        (:Is%20Valid
          _)
        (:Is%20Not%20Valid
          (Variables|SetBy-RefVar (Variables|GetMID%20Variable) _mid))) ; Is the MID valid? If not, re-set it.
      (return false))
    (:Is%20Not%20Valid
      (Utilities|IsValid _returnvalue
        (:Is%20Valid
          (Variables|SetBy-RefVar (Variables|GetMaterial%20Variable) _returnvalue)
          (Variables|SetBy-RefVar (Variables|GetMID%20Variable) _mid)
          (return true))
        (:Is%20Not%20Valid
          (Development|PrintString (Utilities|String|Append "ColorGrass CacheDefaultMaterial FAILED to cache: " (Utilities|GetDisplayName (Variables|GetTarget%20Mesh)) " (" (Utilities|GetDisplayName (Components|GetOwner (Variables|GetTarget%20Mesh))) ")" " , Element Index " (Utilities|String|ToString(Integer) (Variables|GetTarget%20Material%20Index))) false)
          (return false))))) ; Is this material cached yet?
)

(fn SetCurrentColor
  (Variables|Parameters|SetGrassColor_Current (Math|Color|SelectColor (Variables|UserOptions|Basic|GetGrassColor) (Variables|Parameters|GetGrassColor_Default) (Variables|UserOptions|Basic|GetCustomizeGrassColor))))

(fn SetGrassMat
  (Utilities|Array|Add (LocalVariables|GetMIDsToChange) (Variables|Materials|MID|GetGrass_MID))
  (if (!= (Rendering|Material|GetMaterial (Variables|Default|GetCreative_Grass) (Variables|Materials|GetGrassMatIndex)) (Variables|Materials|MID|GetGrass_MID))
    ;[1 1 1] Ensure we also change the material the mesh currently has, which was added by the phone tool
    (Utilities|Array|Add (LocalVariables|GetMIDsToChange) (Material|GetMIDfromMaterial (Variables|Default|GetCreative_Grass) (Variables|Materials|GetGrassMatIndex)))
    (else
      _))
  ;[1 1 1] Grass Color
  (Rendering|Material|SetVectorParameterValue (LocalVariables|GetMIDsToChange) (Variables|Parameters|GetGrassColor_Param) (Variables|Parameters|GetGrassColor_Current) :!class /Script/Engine.MaterialInstanceDynamic))

(fn UpdateVisualsOnClient
  (Default|SetCurrentColor)
  (Default|SetGrassMat))

(fn ConstructionScript)
