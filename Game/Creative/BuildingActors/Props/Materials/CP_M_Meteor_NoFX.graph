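# Creative-prop meteor material, Nanite-enabled, with emissive glow and glitch-push networks left disconnected.
# Object-space triplanar rock and gravel blend by vertex colour, normals and height, then take mask tints, a directional gradient and crystal cracks.
# Reprojected detail normals add to a unique normal map, with optional AO, simpler mobile shading and HLOD proxy fallbacks.
# A previous-frame-switched directional offset feeds WPO before MF_CreativeMoveTool.
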Material /Game/Creative/BuildingActors/Props/Materials/CP_M_Meteor_NoFX
graph {
 properties {bUseMaterialAttributes=true, bUsedWithNanite=true, pixelDepthOffsetMode=PDOM_Legacy, usedShadingModels=Default Lit}
 nodes [
  n1 = TransformPosition(None=WorldPosition().XYZ) {transformSourceType=TRANSFORMPOSSOURCE_World}
  n2 = Transform(None=VertexNormalWS()) {transformSourceType=TRANSFORMSOURCE_World, transformType=TRANSFORM_Local}
  n3 = MaterialFunctionCall!(
   Diffuse Texture Object=TextureObject() {texture=/Game/Athena/Environments/Meteor/Textures/T_Meteor_01_C}
   Spec Texture Object=TextureObject() {texture=/Game/Athena/Environments/Meteor/Textures/T_Meteor_01_S}
   TextureSize=ScalarParameter() {defaultValue=800, parameterName=SizeMeteor_Tiling}
   --------------------------
   Use seperate settings for Z projection
   Use seperate texture for Z projection
   Texture Size Z projection
   Z Plane Diffuse Texture Object
   Z Plane Spec Texture Object
   "--------------------------      "
   ProjectionTransitionContrast=ScalarParameter() {defaultValue=2, parameterName=MeteorProjectionContrast}
   Use Texture Alpha Heightmap to Refine the Transistion
   HeightLerpTransitionContrast
   PickTheDiffuseHeightChannel
   TheDiffuseHeightChannel
   "--------------------------       "
   NormalTextureObject=TextureObject() {samplerType=SAMPLERTYPE_Normal, texture=/Game/Athena/Environments/Meteor/Textures/T_Meteor_01_N}
   World Space Normals=StaticBool()
   UseUncorrectedNormalBlend
   "--------------------------  "
   Export Float 4 Diffuse
   Export Float 4 Spec
   "--------------------------     "
   WorldPosition=@n1
   World Space Normal=@n2
  ) {materialFunction=/Game/Packages/Fortress_SharedMaterials/Base_Material_Functions/WorldAlignedTextures_Complex_OptimzedNormal}
  n4 = MaterialFunctionCall(Float3=@n3.XYZ Spec) {materialFunction=&BreakOutFloat3Components}
  n5 = MakeMaterialAttributes!(
   BaseColor=@n3.XYZ Diffuse Texture
   Metallic=@n4.G
   Specular=@n4.G
   Roughness=Multiply(A=ScalarParameter() {defaultValue=1, parameterName=RoughnessMeteor}, B=@n4.B)
   Anisotropy
   EmissiveColor
   Opacity
   OpacityMask
   Normal=@n3.XYZ Diffuse Texture
   Tangent
   WorldPositionOffset
   SubsurfaceColor
   ClearCoat
   ClearCoatRoughness
   AmbientOcclusion
   Refraction
   CustomizedUVs_0
   CustomizedUVs_1
   CustomizedUVs_2
   CustomizedUVs_3
   CustomizedUVs_4
   CustomizedUVs_5
   CustomizedUVs_6
   CustomizedUVs_7
   PixelDepthOffset
   ShadingModel
   Displacement
  )
  n6 = LinearInterpolate(
   A=ScalarParameter() {defaultValue=-1.2516, parameterName=GrassBlendDirtContrast}
   B=ScalarParameter() {defaultValue=0.8471, parameterName=GrassBlendDirtContrast2}
   Alpha=VertexColor().G
  )
  n7 = ScalarParameter() {defaultValue=1000, parameterName=SizeGravelMeteor_Tiling}
  n8 = MaterialFunctionCall(
   Diffuse Texture Object=TextureObject() {texture=/Game/Athena/Environments/Meteor/Textures/T_Meteor_Gravel_01_C}
   Spec Texture Object=TextureObject() {texture=/Game/Athena/Environments/Meteor/Textures/T_Meteor_Gravel_01_S}
   TextureSize=@n7
   Texture Size Z projection=@n7
   ProjectionTransitionContrast=ScalarParameter() {defaultValue=2, parameterName=MeteorDirtProjectionContrast}
   NormalTextureObject=TextureObject() {samplerType=SAMPLERTYPE_Normal, texture=/Game/Athena/Environments/Meteor/Textures/T_Meteor_Gravel_01_N}
   World Space Normals=StaticBool()
   Export Float 4 Diffuse=StaticBool() {value=true}
   WorldPosition=@n1
   World Space Normal=@n2
  ) {materialFunction=&WorldAlignedTextures_Complex_OptimzedNormal}
  n9 = Multiply(
   A=LinearInterpolate(
    A=VectorParameter() {parameterName=ColorGravel, defaultValue=<rgba 1 1 1 1>}.RGB
    B=VectorParameter() {parameterName=ColorGravel2, defaultValue=<rgba 1 1 1 1>}.RGB
    Alpha=@n6
   )
   B=ComponentMask(None=@n8.XYZ Diffuse Texture) {b=true, g=true, r=true}
  )
  n10 = MaterialFunctionCall(Float3=@n8.XYZ Spec) {materialFunction=&BreakOutFloat3Components}
  n11 = MakeMaterialAttributes(BaseColor=@n9, Metallic=@n10.G, Specular=@n10.G, Roughness=@n10.G, Normal=@n8.Normal XYZ Texture)
  n12 = VertexColor()
  n13 = Add(A=VertexNormalWS(), B=Add(A=Subtract(A=@n12.R) {constB=0.5}, B=Saturate(None=@n6)))
  n14 = MaterialFunctionCall(
   Transition Phase=LinearInterpolate(
    A=ScalarParameter() {parameterName=Grave_Transition_low}
    B=ScalarParameter() {defaultValue=0.8389, parameterName=Gravel_Transition_high}
    Alpha=ComponentMask(None=@n13) {b=true}
   )
   Height Texture=ComponentMask(None=@n8.XYZ Diffuse Texture) {a=true}
   Contrast=ScalarParameter() {defaultValue=-0.019, parameterName=Gravel_TranstionContrast}
  ) {materialFunction=&HeightLerp}
  n15 = TextureSampleParameter2D() {
   parameterName=Mask
   samplerSource=SSM_Wrap_WorldGroupSettings
   samplerType=SAMPLERTYPE_Masks
   texture=/Game/Athena/Environments/Meteor/Textures/Meteor_M
  }
  n16 = LinearInterpolate(
   A=LinearInterpolate(
    A=VectorParameter() {parameterName=Color1, defaultValue=<rgba 0.7656 0.8882 1 1>}.RGB
    B=VectorParameter() {parameterName=Color2, defaultValue=<rgba 0.8197 0.7135 1 1>}.RGB
    Alpha=@n15.G
   )
   B=VectorParameter() {parameterName=Color3, defaultValue=<rgba 1 0.9075 0.7812 1>}.RGB
   Alpha=@n15.B
  )
  n17 = LinearInterpolate(
   A=@n16
   B=Multiply(A=@n16, B=ScalarParameter() {defaultValue=2, parameterName=FresnelBrightness})
   Alpha=Fresnel(ExponentIn=ScalarParameter() {defaultValue=4, parameterName=FresnelMeteor}, Normal=VertexNormalWS()) {baseReflectFraction=0}
  )
  n18 = ShadingPathSwitch(Default=@n17, Deferred=@n17, Forward=@n16, Mobile=@n16)
  n19 = GetMaterialAttributes(None=MaterialFunctionCall!(
    Material=MaterialFunctionCall!(Base Material=@n5, Top Material=@n11, Alpha=@n14.Alpha) {materialFunction=/Engine/Functions/MaterialLayerFunctions/MatLayerBlend_Standard}.Blended Material
    Tint=MaterialProxyReplace(Realtime=@n18, MaterialProxy=@n16)
    TintMask
   ) {materialFunction=/Engine/Functions/MaterialLayerFunctions/MatLayerBlend_Tint}.Blended Material) {attributeGetTypes [
    #1
    #2
    #3
    #4
   ]}
  n20 = TextureSampleParameter2D() {
   parameterName=Mask
   samplerSource=SSM_Wrap_WorldGroupSettings
   samplerType=SAMPLERTYPE_Masks
   texture=&Meteor_M
  }
  n21 = Multiply(A=ComponentMask(None=@n19.Base Color) {r=true}, B=@n20.R)
  n22 = Multiply(
   A=Power(Base=OneMinus(None=@n21), Exp=ScalarParameter() {defaultValue=37.4447, parameterName=AOMaskPower})
   B=ScalarParameter() {defaultValue=1.9946, parameterName=AOMaskBrightness}
  )
  n23 = Clamp(None=LinearInterpolate(
    A=ScalarParameter() {defaultValue=-22.699, parameterName=Cracks_Low}
    B=ScalarParameter() {defaultValue=269.5855, parameterName=Cracks_high}
    Alpha=ComponentMask(None=@n19.Base Color) {r=true}
   )) {clampMode=CMODE_ClampMin}
  n24 = Transform(None=VectorParameter() {parameterName=GradientDirection, defaultValue=<rgba 1 0 -0.5714 1>}.RGB) {transformSourceType=TRANSFORMSOURCE_Local}
  n25 = VectorParameter() {
   desc="Manually making a vector that works for this exact level the hlod is built in.\r\nTransformVector is not working in HLODs"
   parameterName=HLOD_WorldGradientDirection2
   defaultValue=<rgba -0.3 -1 -0.5 1>
  }
  n26 = DotProduct(
   A=Normalize(VectorInput=MaterialProxyReplace(Realtime=@n24, MaterialProxy=@n25.RGB))
   B=Subtract(A=WorldPosition().XYZ, B=ObjectPositionWS())
  )
  n27 = Multiply(
   A=MaterialProxyReplace(Realtime=ObjectRadius(), MaterialProxy=Constant() {r=3000}) {desc=Radius changes during HLOD bake I guess}
   B=ScalarParameter() {defaultValue=2, parameterName=Gradient}
  )
  n28 = Clamp(None=Add(A=Divide(A=@n26, B=@n27), B=ScalarParameter() {defaultValue=0.4215, parameterName=Gradientoffset}) {constB=0.5})
  n29 = OneMinus(None=@n28)
  n30 = ScalarParameter() {defaultValue=29.4328, parameterName=GradientFalloffRear}
  n31 = Multiply(
   A=Multiply(A=Add(A=@n22, B=@n23), B=Power(Base=@n29, Exp=@n30))
   B=VectorParameter() {parameterName=EmissiveRear, defaultValue=<rgba 1 0.1201 0.0604 1>}.RGB
  )
  n32 = OneMinus(None=@n22)
  n33 = ScalarParameter() {defaultValue=22.822, parameterName=GradientFalloffFront}
  n34 = Multiply(
   A=Multiply(A=Add(A=@n32, B=@n23), B=Power(Base=@n28, Exp=@n33))
   B=VectorParameter() {parameterName=EmissiveFront, defaultValue=<rgba 20 0.238 0 1>}.RGB
  )
  n35 = VectorParameter() {parameterName=EmissiveCracksBlue, defaultValue=<rgba 0.604 1.0675 5 1>}
  n36 = VertexColor()
  n37 = ScalarParameter() {defaultValue=0.01, parameterName=CracksBrightness2}
  n38 = MaterialFunctionCall(
   TextureObject=TextureObject() {texture=/Game/Athena/Environments/Meteor/Textures/Meteor_Cracks_M}
   TextureSize=ScalarParameter() {defaultValue=1000, parameterName=CrackSize}
   WorldPosition=@n1
   World Space Normal=@n2
  ) {materialFunction=&WorldAlignedTexture}
  n39 = ComponentMask(None=@n38.XYZ Texture) {b=true}
  n40 = Multiply(
   A=Power(Base=@n39, Exp=ScalarParameter() {defaultValue=5, parameterName=CracksPower})
   B=ScalarParameter() {defaultValue=50, parameterName=CracksBrightness}
  )
  n41 = Multiply(A=@n36.B, B=Add(A=Multiply(A=@n22, B=@n37), B=Multiply(A=@n22, B=@n40)))
  n42 = MaterialFunctionCall!(MaskScaleMult) {materialFunction=/Game/Athena/Effects/Materials/Meteor/MF_Glitch_FX_World_Mask}
  n43 = Saturate(None=LinearInterpolate(Alpha=@n42.DistanceFadeMask) {constA=-0.5})
  n44 = VectorParameter() {parameterName=Chroma1, defaultValue=<rgba 0.0499 0.6782 2 0>}
  n45 = VectorParameter() {parameterName=Chroma2, defaultValue=<rgba 0.0078 0.2475 0.25 0>}
  n46 = Fresnel() {baseReflectFraction=0, exponent=7}
  n47 = Multiply(A=@n42.WorldSpaceMask, B=Multiply(A=LinearInterpolate(A=@n44.RGB, B=@n45.RGB, Alpha=@n46), B=@n46))
  n48 = Constant()
  Add(A=Add(A=Add(A=@n31, B=@n34), B=Multiply(A=@n35.RGB, B=@n41)), B=QualitySwitch(Default=Multiply(A=@n43, B=@n47), Low=@n48, Medium=@n48))
  n49 = Constant() {r=1}
  n50 = Clamp(None=Add(A=ComponentMask(None=@n38.XYZ Texture) {g=true}, B=ScalarParameter() {parameterName=CrackShadow_Brightness}))
  n51 = Clamp(None=Multiply(A=@n36.B, B=ScalarParameter() {defaultValue=8, parameterName=CrackshadowBias}))
  n52 = VectorParameter() {parameterName=TintRear, defaultValue=<rgba 0.01 0.0006 0.0004 1>}
  n53 = VectorParameter() {parameterName=TintFornt, defaultValue=<rgba 1.5 1.5 1.5 1>}
  n54 = Multiply(A=@n19.Base Color, B=Multiply(A=LinearInterpolate(A=@n52.RGB, B=@n53.RGB, Alpha=@n28), B=LinearInterpolate(A=@n49, B=@n50, Alpha=@n51)))
  n55 = VectorParameter() {parameterName=CrystalColor, defaultValue=<rgba 0.0136 0.0169 0.1823 1>}
  n56 = Multiply(A=Clamp(None=Multiply(A=@n39) {constA=3, constB=4}), B=@n36.B)
  n57 = SetMaterialAttributes!(
   MaterialAttributes=@n19.MaterialAttributes
   Base Color=MaterialProxyReplace(Realtime=LinearInterpolate(A=@n54, B=@n55.RGB, Alpha=@n56), MaterialProxy=@n19.Base Color)
   Emissive Color
   Roughness=LinearInterpolate(A=@n19.Roughness, B=Constant(), Alpha=@n56)
   Specular=LinearInterpolate(A=@n19.Specular, B=Constant() {r=1}, Alpha=@n56)
   Normal=LinearInterpolate(A=@n19.Normal, B=Constant3Vector() {constant=<rgba 0 0 1 0>}, Alpha=@n56)
  ) {attributeSetTypes [
    #1
    #5
    #2
    #3
    #4
   ]}
  n58 = GetMaterialAttributes(None=@n57) {attributeGetTypes [#4]}
  n59 = MaterialFunctionCall(Float3=@n58.Normal) {materialFunction=&BreakOutFloat3Components}
  n60 = Multiply(A=@n59.R) {constB=-1}
  n61 = Multiply(A=Transform(None=Constant3Vector() {constant=<rgba 0 0 1 0>}), B=Constant() {r=1})
  n62 = Multiply(A=@n59.B, B=ComponentMask(None=@n61) {g=true})
  n63 = ComponentMask(None=@n61) {r=true}
  n64 = Multiply(A=@n63, B=@n59.B)
  n65 = Multiply(A=@n59.R) {constB=-1}
  n66 = ScalarParameter() {parameterName=ContrastBlendNormals}
  n67 = LinearInterpolate(
   A=AppendVector(A=AppendVector(A=@n60, B=@n62), B=Multiply(A=@n59.G) {constB=-1})
   B=AppendVector(A=AppendVector(A=@n64, B=@n65), B=Multiply(A=@n59.G) {constB=-1})
   Alpha=MaterialFunctionCall(In=Abs(None=@n63), Contrast=@n66) {materialFunction=&CheapContrast}.Result
  )
  n68 = Constant3Vector() {constant=<rgba -1 -1 1 0>}
  n69 = ComponentMask(None=@n61) {b=true}
  n70 = AppendVector(A=ComponentMask(None=Multiply(A=@n58.Normal, B=@n68)) {g=true, r=true}, B=Multiply(A=@n59.B, B=@n69))
  n71 = MaterialFunctionCall(In=Abs(None=@n69), Contrast=@n66) {materialFunction=&CheapContrast}
  n72 = Transform(None=Normalize(VectorInput=LinearInterpolate(A=@n67, B=@n70, Alpha=@n71.Result))) {transformSourceType=TRANSFORMSOURCE_World, transformType=TRANSFORM_Tangent}
  AppendVector(A=ComponentMask(None=@n72) {g=true, r=true})
  Constant3Vector() {constant=<rgba 0 0 1 0>}
  Constant3Vector()
  Constant()
  n73 = Multiply(A=@n72, B=VectorParameter() {parameterName=NormalIntensity, defaultValue=<rgba 1 1 0 1>}.RGB)
  n74 = TextureSampleParameter2D() {parameterName=Normal, samplerType=SAMPLERTYPE_Normal, texture=/Game/Athena/Environments/Meteor/Textures/Meteor_N}
  n75 = SetMaterialAttributes!(MaterialAttributes=@n58.MaterialAttributes, Normal=MaterialProxyReplace(Realtime=Add(A=@n74.RGB, B=@n73), MaterialProxy=@n58.Normal)) {attributeSetTypes [#4]}
  n76 = GetMaterialAttributes(None=@n75) {attributeGetTypes [#1]}
  n77 = TextureSampleParameter2D() {
   parameterName=Mask
   samplerSource=SSM_Wrap_WorldGroupSettings
   samplerType=SAMPLERTYPE_Masks
   texture=&Meteor_M
  }
  n78 = ScalarParameter() {defaultValue=0.5, parameterName=AOintensity}
  n79 = Multiply(A=@n76.Base Color, B=Saturate(None=Add(A=@n77.R, B=@n78)))
  n80 = StaticSwitchParameter(
   True=SetMaterialAttributes!(MaterialAttributes=@n76.MaterialAttributes, Base Color=MaterialProxyReplace(Realtime=@n79, MaterialProxy=@n76.Base Color)) {attributeSetTypes [#1]}
   False=@n75
  ) {defaultValue=true, parameterName=UseAO}
  n81 = SetMaterialAttributes!(
   MaterialAttributes=GetMaterialAttributes(None=@n75).MaterialAttributes
   Specular=ScalarParameter() {defaultValue=0.1, parameterName=MobilSpec}
   Roughness=ScalarParameter() {defaultValue=0.8, parameterName=MobilRoughness}
  ) {attributeSetTypes [#3, #2]}
  n82 = VectorParameter() {parameterName=MeteorDirection, defaultValue=<rgba 0.7755 -0.4565 -0.4361 0>}
  n83 = Multiply(B=@n82.RGB) {constA=-1}
  MaterialFunctionCall(MA In=MaterialFunctionCall!(
    Material=ShadingPathSwitch(Default=@n80, Deferred=@n80, Forward=@n81, Mobile=@n81)
    WorldPositionOffset=PreviousFrameSwitch(Current Frame=@n83, Previous Frame=@n82.RGB)
   ) {materialFunction=/Engine/Functions/MaterialLayerFunctions/MatLayerBlend_AddWorldPositionOffset}.Blended Material) {materialFunction=&MF_CreativeMoveTool}
  Multiply(B=@n9)
  Normalize(VectorInput=@n72)
  n84 = MaterialFunctionCall() {materialFunction=&MF_Glitch_FX_World_Mask}
  n85 = LinearInterpolate(Alpha=@n84.DistanceFadeMask) {constB=3}
  n86 = Multiply(A=@n82.RGB, B=DotProduct(A=@n82.RGB, B=VertexNormalWS()))
  n87 = Multiply(A=Multiply(A=@n84.WorldSpaceMask, B=Multiply(A=@n85, B=@n86)), B=ScalarParameter() {defaultValue=100, parameterName=Push})
  n88 = Constant()
  QualitySwitch(Default=ShadingPathSwitch(Default=FeatureLevelSwitch(Default=@n87, ES2=@n88, ES3_1=@n88, SM4_REMOVED=@n87), Forward=@n88), Low=@n88)
  ScalarParameter() {defaultValue=0.0857, parameterName=SpecularDirt}
  VertexNormalWS()
 ]
}
